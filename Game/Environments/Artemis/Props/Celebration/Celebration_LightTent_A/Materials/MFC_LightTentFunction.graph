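# Material function that adds animated emissive lighting to incoming material attributes for the celebration light tent prop, toggled by a static switch.
# A panning color texture and a tinted, tiled micro mask are blended by that mask, with inverted normal-based Fresnel driving both strengths.
# A time-of-day function scales the glow, which adds to existing emissive and blends in through a grayscale mask.
# The SphereMask branch is unconnected.
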MaterialFunction /Game/Environments/Artemis/Props/Celebration/Celebration_LightTent_A/Materials/MFC_LightTentFunction
graph {nodes [
  n1 = GetMaterialAttributes(None=FunctionInput() {inputName=Material, inputType=FunctionInput_MaterialAttributes}) {attributeGetTypes [#1, #2]}
  n2 = LinearInterpolate(
   A=TextureCoordinate()
   B=TextureCoordinate() {coordinateIndex=2}
   Alpha=ScalarParameter() {defaultValue=1, group=LightTent, parameterName=LT_UseUV3Emissive}
  )
  n3 = AppendVector(
   A=ScalarParameter() {defaultValue=1, group=LightTent, parameterName=LT_EmissiveColorTextureTile_X}
   B=ScalarParameter() {defaultValue=1, group=LightTent, parameterName=LT_EmissiveColorTextureTile_Y}
  )
  n4 = Time()
  n5 = AppendVector(
   A=ScalarParameter() {defaultValue=0.25, group=LightTent, parameterName=LT_EmissiveColorTexturePanSpeed_X}
   B=ScalarParameter() {group=LightTent, parameterName=LT_EmissiveColorTexturePanSpeed_Y}
  )
  n6 = TextureSampleParameter2D(UVs=Add(A=Multiply(A=@n2, B=@n3), B=Multiply(A=@n4, B=@n5))) {group=LightTent, parameterName=LT_EmissiveTexture, texture=/Game/Athena/Artemis/Maps/Artemis_GradientTest_01}
  n7 = Fresnel(
   ExponentIn=ScalarParameter() {defaultValue=1, group=LightTent, parameterName=LT_FresnelExponent}
   Normal=Normalize(VectorInput=Transform(None=@n1.Normal))
  )
  n8 = Saturate(None=Power(Base=OneMinus(None=@n7), Exp=ScalarParameter() {defaultValue=2, group=LightTent, parameterName=LT_FresnelPower}))
  n9 = LinearInterpolate(
   A=ScalarParameter() {defaultValue=1, group=LightTent, parameterName=LT_EmissiveStrength_Min}
   B=ScalarParameter() {defaultValue=5, group=LightTent, parameterName=LT_EmissiveStrength_MAx}
   Alpha=@n8
  )
  n10 = Multiply(
   A=LinearInterpolate(
    A=TextureCoordinate()
    B=TextureCoordinate() {coordinateIndex=2}
    Alpha=ScalarParameter() {defaultValue=1, group=LightTent, parameterName=LT_UseUV3Emissive}
   )
   B=AppendVector(
    A=ScalarParameter() {defaultValue=64, group=LightTent, parameterName=LT_EmissiveMicroColorTextureTile_X, sliderMin=64}
    B=ScalarParameter() {defaultValue=64, group=LightTent, parameterName=LT_EmissiveMicroColorTextureTile_Y, sliderMin=64}
   )
  )
  n11 = Multiply(
   A=Time()
   B=AppendVector(
    A=ScalarParameter() {group=LightTent, parameterName=LT_EmissiveMicroColorTexturePanSpeed_X}
    B=ScalarParameter() {defaultValue=3, group=LightTent, parameterName=LT_EmissiveMicroColorTexturePanSpeed_Y}
   )
  )
  n12 = ChannelMaskParameter(None=TextureSampleParameter2D(UVs=Add(A=@n10, B=@n11)) {group=LightTent, parameterName=LT_EmissiveMicroTexture, texture=/Game/Effects/Fort_Effects/Textures/Mask/T_HalfSphere_D}.RGB) {maskChannel=Green, parameterName=LT_MicroTextureChannel, defaultValue=<rgba 0 1 0 0>}
  n13 = Multiply(A=@n12.R, B=VectorParameter() {parameterName=LT_MicroEmissiveColor, defaultValue=<rgba 1 1 1 1>}.RGB)
  n14 = LinearInterpolate(
   A=ScalarParameter() {defaultValue=1, group=LightTent, parameterName=LT_MicroEmissiveStrength_Min}
   B=ScalarParameter() {defaultValue=10, group=LightTent, parameterName=LT_MicroEmissiveStrength_MAx}
   Alpha=@n8
  )
  n15 = LinearInterpolate(
   A=Power(Base=Multiply(A=@n6.RGB, B=@n9), Exp=ScalarParameter() {defaultValue=2, group=LightTent, parameterName=LT_EmissivePower})
   B=Power(Base=Multiply(A=@n13, B=@n14), Exp=ScalarParameter() {defaultValue=2, group=LightTent, parameterName=LT_EmissivePower})
   Alpha=@n12.R
  )
  n16 = MaterialFunctionCall(
   Color_Morning=ScalarParameter() {defaultValue=1, parameterName=LT_EmissiveMorningMult}
   Color_Day=ScalarParameter() {defaultValue=1, parameterName=LT_EmissiveDayMult}
   Color_Evening=ScalarParameter() {defaultValue=1, parameterName=LT_EmissiveEveningMult}
   Color_Night=ScalarParameter() {defaultValue=1, parameterName=LT_EmissiveNightMult}
  ) {materialFunction=&Color_Time_of_Day}
  n17 = SetMaterialAttributes!(MaterialAttributes=@n1.MaterialAttributes, EmissiveColor=Add(A=@n1.EmissiveColor, B=Multiply(A=@n15, B=@n16.Result))) {attributeSetTypes [#1]}
  n18 = TextureSampleParameter2D(UVs=@n2) {
   group=LightTent
   parameterName=LT_EmissiveMask
   samplerType=SAMPLERTYPE_Grayscale
   texture=/Game/Environments/Artemis/Props/Celebration/Celebration_LightTent_A/Textures/T_Celebration_LightTentPetal_A_A_E
  }
  FunctionOutput(None=StaticSwitchParameter(True=BlendMaterialAttributes(A=@n1.MaterialAttributes, B=@n17, Alpha=@n18.R) {vertexAttributeBlendType=UseA}, False=@n1.MaterialAttributes) {parameterName=UseLightTentFunction})
  n19 = VectorParameter() {parameterName=MaskPos, defaultValue=<rgba 1 0 0 1>}
  OneMinus(None=SphereMask(A=@n2, B=ComponentMask(None=@n19.RGB) {g=true, r=true}, Radius=Constant() {r=0.25}))
  PixelNormalWS()
 ]}
